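# Sound cue for the Rufus StoneHead spawn sound in the JunkRift content.
# A Random node picks one of four spawn wave variations with equal weight, feeding a Modulator that slightly randomizes pitch.
# The cue plays at reduced volume in the gameplay events sound class, with overridden attenuation that sets the stereo spread.
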SoundCue /BlastBerryMapContent/Game/Blueprints/JunkRift/SFX/Rufus_StoneHead_Spawn_Cue
root=Modulator_0
nodes [
 {id=Modulator_0, children [Random_0], properties {pitchMax=1.1, pitchMin=0.9}}
 {id=Random_0, children [WavePlayer_0,WavePlayer_1,WavePlayer_2,WavePlayer_3], properties {weights [1,1,1,1]}}
 {id=WavePlayer_0, properties {soundWaveAssetPtr=/BlastberryMapContent/Game/Blueprints/JunkRift/SFX/Rufus_StoneHead_Spawn_01}}
 {id=WavePlayer_1, properties {soundWaveAssetPtr=/BlastberryMapContent/Game/Blueprints/JunkRift/SFX/Rufus_StoneHead_Spawn_02}}
 {id=WavePlayer_2, properties {soundWaveAssetPtr=/BlastberryMapContent/Game/Blueprints/JunkRift/SFX/Rufus_StoneHead_Spawn_03}}
 {id=WavePlayer_3, properties {soundWaveAssetPtr=/BlastberryMapContent/Game/Blueprints/JunkRift/SFX/Rufus_StoneHead_Spawn_04}}
]
properties {
 bOverrideAttenuation=true
 maxDistance=4000
 soundClassObject=&Fort_Events_Gameplay
 volumeMultiplier=0.376
 attenuationOverrides {stereoSpread=512}
}
